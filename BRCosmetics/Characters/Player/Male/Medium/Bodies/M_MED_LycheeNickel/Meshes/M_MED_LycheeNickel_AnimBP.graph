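# Animation Blueprint for the M_MED_LycheeNickel cosmetic body that follows its attached parent mesh and adds secondary motion to the collar and hoodie.
# Copy Pose From Mesh brings in pose and curves, then feeds the body's Control Rig.
# Chained spring-based Anim Dynamics nodes then simulate collar chain, medal and first-level hoodie joints.
# Second-level hoodie dynamics and a Retarget Pose From Mesh node are unconnected, and no event logic runs.
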AnimBlueprint /BRCosmetics/Characters/Player/Male/Medium/Bodies/M_MED_LycheeNickel/Meshes/M_MED_LycheeNickel_AnimBP
target_skeleton=&/M_MED_LycheeNickel_Skeleton
graphs [
 AnimGraph {nodes [
   Anim Dynamics - Bone: dyn_hoodie_Ct_02_jnt(Alpha=1, ExternalForce="0,0,0") {
    angularDampingOverride=0.9999
    angularSpringConstant=2000
    bAngularSpring=true
    bOverrideAngularDamping=true
    bUseGravityOverride=true
    simulationSpace=RootRelative
    alphaBoolBlend {blendOption=Linear}
    boundBone {boneName=dyn_hoodie_Ct_02_jnt}
    componentLinearAccScale=<0.4 0.4 0.4>
    componentLinearVelScale=<0.4 0.4 0.4>
    gravityOverride {z=-980}
    physicsBodyDefinitions [{
      boundBone {boneName=dyn_hoodie_Ct_02_jnt}
      boxExtents=<20 20 20>
      constraintSetup {angularLimitsMax=<5 1 25>, angularLimitsMin=<-5 -1 -25>, angularTarget=<1 0 0>}
      localJointOffset=<2 4 0>
     }]
   }
   Anim Dynamics - Bone: dyn_hoodie_Rt_02_jnt(Alpha=1, ExternalForce="0,0,0") {
    angularDampingOverride=0.9999
    angularSpringConstant=2000
    bAngularSpring=true
    bOverrideAngularDamping=true
    bUseGravityOverride=true
    simulationSpace=RootRelative
    alphaBoolBlend {blendOption=Linear}
    boundBone {boneName=dyn_hoodie_Rt_02_jnt}
    componentLinearAccScale=<0.4 0.4 0.4>
    componentLinearVelScale=<0.4 0.4 0.4>
    gravityOverride {z=-980}
    physicsBodyDefinitions [{
      boundBone {boneName=dyn_hoodie_Rt_02_jnt}
      boxExtents=<20 20 20>
      constraintSetup {angularLimitsMax=<5 1 25>, angularLimitsMin=<-5 -1 -25>, angularTarget=<1 0 0>}
      localJointOffset=<-2 -4 0>
     }]
   }
   Anim Dynamics - Bone: dyn_hoodie_Lf_02_jnt(Alpha=1, ExternalForce="0,0,0") {
    angularDampingOverride=0.9999
    angularSpringConstant=2000
    bAngularSpring=true
    bOverrideAngularDamping=true
    bUseGravityOverride=true
    simulationSpace=RootRelative
    alphaBoolBlend {blendOption=Linear}
    boundBone {boneName=dyn_hoodie_Lf_02_jnt}
    componentLinearAccScale=<0.4 0.4 0.4>
    componentLinearVelScale=<0.4 0.4 0.4>
    gravityOverride {z=-980}
    physicsBodyDefinitions [{
      boundBone {boneName=dyn_hoodie_Lf_02_jnt}
      boxExtents=<20 20 20>
      constraintSetup {angularLimitsMax=<5 1 25>, angularLimitsMin=<-5 -1 -25>, angularTarget=<1 0 0>}
      localJointOffset=<2 4 0>
     }]
   }
   Retarget Pose From Mesh!() -> (Pose:Pose Link Structure) {
    is=title AnimGraphNode_RetargetPoseFromMesh
    iKRetargeterAsset=&/M_MED_LycheeNickel_NewRetargeter
   }
   n1 = Control Rig(Source=Copy Pose From Mesh() {bCopyCurves=true, bUseAttachedParent=true}) {
    alphaBoolBlend {blendOption=Linear}
    controlRigAssetReference {blueprintRigClass=&/M_MED_LycheeNickel_CtrlRig.M_MED_LycheeNickel_CtrlRig_C}
   }
   n2 = Anim Dynamics - Bone: dyn_collar_chain_jnt(ComponentPose=Local To Component(LocalPose=@n1), Alpha=0.5, ExternalForce="0,0,0") {
    angularDampingOverride=0.9999
    angularSpringConstant=200
    bAngularSpring=true
    bOverrideAngularDamping=true
    bUseGravityOverride=true
    simulationSpace=RootRelative
    alphaBoolBlend {blendOption=Linear}
    boundBone {boneName=dyn_collar_chain_jnt}
    componentLinearAccScale=<0.4 0.4 0.4>
    componentLinearVelScale=<0.4 0.4 0.4>
    gravityOverride {z=-980}
    physicsBodyDefinitions [{
      boundBone {boneName=dyn_collar_chain_jnt}
      boxExtents=<10 10 10>
      constraintSetup {angularLimitsMax=<1 10 0>, angularLimitsMin=<-1 -10 0>, angularTarget=<1 0 0>}
      localJointOffset=<2 0 0>
     }]
   }
   n3 = Anim Dynamics - Bone: dyn_collar_medal_jnt(ComponentPose=@n2, Alpha=1, ExternalForce="0,0,0") {
    angularDampingOverride=0.9999
    angularSpringConstant=200
    bAngularSpring=true
    bOverrideAngularDamping=true
    bUseGravityOverride=true
    simulationSpace=RootRelative
    alphaBoolBlend {blendOption=Linear}
    boundBone {boneName=dyn_collar_medal_jnt}
    componentLinearAccScale=<0.4 0.4 0.4>
    componentLinearVelScale=<0.4 0.4 0.4>
    gravityOverride {z=-980}
    physicsBodyDefinitions [{
      boundBone {boneName=dyn_collar_medal_jnt}
      boxExtents=<10 10 10>
      constraintSetup {angularLimitsMax=<1 10 0>, angularLimitsMin=<-1 -10 -10>, angularTarget=<1 0 0>}
      localJointOffset=<2 0 0>
     }]
   }
   n4 = Anim Dynamics - Bone: dyn_hoodie_Rt_01_jnt(ComponentPose=@n3, Alpha=1, ExternalForce="0,0,0") {
    angularDampingOverride=0.9999
    angularSpringConstant=50
    bAngularSpring=true
    bOverrideAngularDamping=true
    bUseGravityOverride=true
    simulationSpace=RootRelative
    alphaBoolBlend {blendOption=Linear}
    boundBone {boneName=dyn_hoodie_Rt_01_jnt}
    componentLinearAccScale=<0.4 0.4 0.4>
    componentLinearVelScale=<0.4 0.4 0.4>
    gravityOverride {z=-980}
    physicsBodyDefinitions [{
      boundBone {boneName=dyn_hoodie_Rt_01_jnt}
      boxExtents=<20 20 20>
      constraintSetup {angularLimitsMax=<1 5 0>, angularLimitsMin=<-1 -5 -10>, angularTarget=<1 0 0>}
      localJointOffset=<-10 0 0>
     }]
   }
   n5 = Anim Dynamics - Bone: dyn_hoodie_Lf_01_jnt(ComponentPose=@n4, Alpha=1, ExternalForce="0,0,0") {
    angularDampingOverride=0.9999
    angularSpringConstant=50
    bAngularSpring=true
    bOverrideAngularDamping=true
    bUseGravityOverride=true
    simulationSpace=RootRelative
    alphaBoolBlend {blendOption=Linear}
    boundBone {boneName=dyn_hoodie_Lf_01_jnt}
    componentLinearAccScale=<0.4 0.4 0.4>
    componentLinearVelScale=<0.4 0.4 0.4>
    gravityOverride {z=-980}
    physicsBodyDefinitions [{
      boundBone {boneName=dyn_hoodie_Lf_01_jnt}
      boxExtents=<20 20 20>
      constraintSetup {angularLimitsMax=<1 5 0>, angularLimitsMin=<-1 -5 -10>, angularTarget=<1 0 0>}
      localJointOffset=<10 0 0>
     }]
   }
   n6 = Anim Dynamics - Bone: dyn_hoodie_Ct_01_jnt(ComponentPose=@n5, Alpha=1, ExternalForce="0,0,0") {
    angularDampingOverride=0.9999
    angularSpringConstant=50
    bAngularSpring=true
    bOverrideAngularDamping=true
    bUseGravityOverride=true
    simulationSpace=RootRelative
    alphaBoolBlend {blendOption=Linear}
    boundBone {boneName=dyn_hoodie_Ct_01_jnt}
    componentLinearAccScale=<0.4 0.4 0.4>
    componentLinearVelScale=<0.4 0.4 0.4>
    gravityOverride {z=-980}
    physicsBodyDefinitions [{
      boundBone {boneName=dyn_hoodie_Ct_01_jnt}
      boxExtents=<20 20 20>
      constraintSetup {angularLimitsMax=<1 5 0>, angularLimitsMin=<-1 -5 -35>, angularTarget=<1 0 0>}
      localJointOffset=<10 0 0>
     }]
   }
   Output Pose(Result=Component To Local(ComponentPose=@n6)) {layerGroup=None}
  ]}
 EventGraph {nodes [Event BlueprintUpdateAnimation()]}
]
variables []
